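# M_GasPumpMetalBall is a Default Lit material in the Commerce set props, named for the metal ball of a gas pump.
# It holds only a grey color constant and two scalar constants, with no textures or parameters; their output connections are not shown.
# Usage flags enable it for skeletal meshes, mesh deformers and Nanite.
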Material /Game/Environments/Apollo/Sets/Commerce/Props/Materials/M_GasPumpMetalBall
graph {
 properties {
  bUsedWithMeshDeformer=true
  bUsedWithNanite=true
  bUsedWithSkeletalMesh=true
  pixelDepthOffsetMode=PDOM_Legacy
  usedShadingModels=Default Lit
 }
 nodes [
  Constant3Vector() {constant=<rgba 0.4635 0.4635 0.4635 0>}
  Constant() {r=1}
  Constant() {r=0.4}
 ]
}
